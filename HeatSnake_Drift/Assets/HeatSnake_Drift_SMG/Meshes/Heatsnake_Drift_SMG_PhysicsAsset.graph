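# Physics asset for the Heatsnake Drift SMG skeletal mesh, giving the weapon its simplified collision shape.
# A single body on the handle bone holds one capsule, offset and rotated from the bone origin.
# It has no constraints, so the weapon simulates as one rigid piece rather than articulated parts.
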PhysicsAsset /HeatSnake_Drift/Assets/HeatSnake_Drift_SMG/Meshes/Heatsnake_Drift_SMG_PhysicsAsset
bodies {handle [{shapeType=Capsule, length=25.9818, radius=13.6482, center=<6.5801 -0.0004 3.0677>, rotation=<rot 0 90.002 -81.0261>}]}
constraints []
